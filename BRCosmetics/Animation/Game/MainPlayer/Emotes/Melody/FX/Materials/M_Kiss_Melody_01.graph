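# Translucent unlit Niagara sprite material that draws and animates the kiss texture for the Melody emote, driven by particle dynamic parameters.
# UVs get a radial sphere-mask distortion, while a slightly rotated horizontal split shifts the halves together for a squash.
# A powered sine tints the seam dark red before particle color and emissive adjustment.
# A sine-based horizontal mask trims alpha, taken from texture red or alpha and scaled by particle alpha.
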Material /BRCosmetics/Animation/Game/MainPlayer/Emotes/Melody/FX/Materials/M_Kiss_Melody_01
graph {
 properties {
  bAutomaticallySetUsageInEditor=false
  bUsedWithNiagaraSprites=true
  blendMode=BLEND_Translucent
  pixelDepthOffsetMode=PDOM_Legacy
  shadingModel=MSM_Unlit
  usedShadingModels=Unlit
 }
 nodes [
  Constant() {r=0.35}
  n1 = ParticleColor()
  n2 = TextureCoordinate()
  n3 = Constant() {r=0.5}
  n4 = DynamicParameter() {defaultValue=<rgba 0.35 1 1 0>, paramNames [Distortion_Radius, Distortion_Mult, Shrink_Mask, UV_Shrink]}
  n5 = Constant()
  n6 = TextureCoordinate()
  n7 = Multiply(
   A=Multiply(
    A=SphereMask(A=@n2, B=@n3, Radius=@n4.Distortion_Radius, Hardness=@n5) {hardnessPercent=0}
    B=ConstantBiasScale(None=@n6) {bias=-0.5, scale=2}
   )
   B=Multiply(A=@n4.Distortion_Mult, B=ScalarParameter() {defaultValue=1, parameterName=Distortion_Mult})
  ) {constB=0.104}
  n8 = TextureCoordinate()
  n9 = TextureCoordinate()
  n10 = Constant() {r=-0.003}
  n11 = Add(A=ComponentMask(None=MaterialFunctionCall(UVs=@n9, "Rotation Angle (0-1)"=@n10) {materialFunction=&CustomRotator}.Rotated Values) {g=true}) {constB=-0.027}
  n12 = Saturate(None=Round(None=@n11))
  n13 = AppendVector(A=Constant(), B=Multiply(A=@n12, B=@n4.UV_Shrink) {constB=0.048})
  n14 = Constant()
  n15 = Multiply(A=OneMinus(None=@n12), B=@n4.UV_Shrink)
  n16 = Constant() {r=1}
  n17 = MaterialFunctionCall(UVs=Subtract(A=Add(A=@n8, B=@n13), B=AppendVector(A=@n14, B=@n15)), Texture Scale=AppendVector(A=@n16, B=@n16)) {materialFunction=&ScaleUVsByCenter}
  n18 = TextureSampleParameter2D(UVs=Saturate(None=Add(A=@n7, B=@n17.UVs))) {parameterName=KissTexture, texture=/BRCosmetics/Animation/Game/MainPlayer/Emotes/Melody/FX/Textures/T_Kiss_Melody_01}
  n19 = Add(A=TextureCoordinate()) {constB=0.01}
  n20 = Add(A=Sine(None=ComponentMask(None=@n19) {r=true}) {period=2}, B=@n4.Shrink_Mask) {constB=0.4206}
  n21 = Saturate(None=Ceil(None=Round(None=@n20)))
  LinearInterpolate(
   A=Multiply(A=Multiply(A=@n1.A, B=@n18.R), B=@n21)
   B=Multiply(A=Multiply(A=@n18.A, B=@n21), B=@n1.A)
   Alpha=ScalarParameter() {defaultValue=1, parameterName=UseTextureAlpha}
  )
  n22 = Constant3Vector() {constant=<rgba 0.2969 0 0.0004 1>}
  n23 = Power(Base=Sine(None=@n11) {period=2}) {constExponent=600}
  n24 = Multiply(
   A=Multiply(A=Saturate(None=@n23), B=DynamicParameter() {parameterIndex=1, paramNames [Shrink_Color_Mask, Param2, Param3, Param4]}.Shrink_Color_Mask)
   B=ScalarParameter() {defaultValue=1, parameterName=Shrink_Color_Mask}
  )
  MaterialFunctionCall(Emissive=Multiply(A=LinearInterpolate(A=@n18.RGB, B=@n22, Alpha=@n24), B=@n1.RGB)) {materialFunction=&MF_FX_EmissiveAdjustments}
 ]
}
