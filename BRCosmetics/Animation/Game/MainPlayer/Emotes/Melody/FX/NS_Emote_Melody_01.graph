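# Niagara effect for the Melody emote that sends out a large kiss sprite with smaller kisses, sparkles and a light-ray flash.
# Kiss sprites animate squash-and-stretch and material parameters over life, with a second renderer and red glow parameters.
# All four emitters conditionally inherit source movement by comparing owner speed to a threshold.
# Front-end switches adjust kiss velocity, glow and sparkle alpha, while time-of-day scalars adjust kiss brightness.
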NiagaraSystem /BRCosmetics/Animation/Game/MainPlayer/Emotes/Melody/FX/NS_Emote_Melody_01
dependencies {
 usedDataInterfaces [
  &NiagaraDataInterfaceCurve
  &NiagaraDataInterfaceVector2DCurve
  &NiagaraDataInterfaceColorCurve
  &NiagaraDataInterfaceMeshRendererInfo
 ]
 usedDynamicInputs []
 usedModules [&SystemState]
}
emitters {
 KissBig {
  emitterUpdateScript [
   &EmitterState()
   &SpawnBurst_Instantaneous(Spawn Group=1, Use Spawn Probability=false, UseLoopCountLimit=false)
  ]
  particleSpawnScript [
   &InitializeParticle(
    Mass Mode=::NewEnumerator3
    Position Offset=<0 0 5>
    Sprite Size Mode=::NewEnumerator3
    Uniform Sprite Size=33
    UsePositionOffset=true
    Write Material Random=true
   )
   &AddVelocity(Velocity=&FN_FrontEndVector3Modifier(Gameplay=<0 300 0>, Front End=<-50 300 -11>))
   /SetVariables(Particles.GlowCameraOffset=-0.1) {isSetParameters=true}
  ]
  particleUpdateScript [
   &ParticleState()
   &ScaleSpriteSize(
    Non-Uniform Curve Scale=<0.8 0.8>
    Non-Uniform Curve Sprite Scale=DataInterface() {propertyValues {
      XCurve {keys [
        <at 0 0.8 RCIM_Cubic>
        {arriveTangent=-2.123, interpMode=RCIM_Cubic, leaveTangent=-2.123, tangentMode=RCTM_User, time=0.1, value=0.9876}
        {arriveTangent=0.432, interpMode=RCIM_Cubic, leaveTangent=0.432, tangentMode=RCTM_User, time=0.3641, value=0.6595}
        {arriveTangent=0.4905, interpMode=RCIM_Cubic, leaveTangent=0.4905, tangentMode=RCTM_User, time=0.57, value=0.4961}
        {arriveTangent=0.9343, interpMode=RCIM_Cubic, leaveTangent=0.9343, tangentMode=RCTM_User, time=0.8, value=0.8278}
        {interpMode=RCIM_Cubic, leaveTangent=-0.1918, time=1, value=0.9223}
       ]}
      YCurve {keys [
        {arriveTangent=1.4872, interpMode=RCIM_Cubic, leaveTangent=1.4872, tangentMode=RCTM_User, time=0, value=1.1}
        {interpMode=RCIM_Cubic, tangentMode=RCTM_User, time=0.3279, value=1.2}
        {arriveTangent=-0.4468, interpMode=RCIM_Cubic, leaveTangent=-0.4468, time=0.5326, value=1.4335}
        {arriveTangent=-2.0355, interpMode=RCIM_Cubic, leaveTangent=-2.0355, tangentMode=RCTM_User, time=0.6917, value=1.0375}
        {arriveTangent=-0.6449, interpMode=RCIM_Cubic, leaveTangent=-0.6449, tangentMode=RCTM_User, time=0.8279, value=0.85}
        {interpMode=RCIM_Cubic, tangentMode=RCTM_User, time=1, value=0.8658}
       ]}
     }}
    Scale Sprite Size Mode=::NewEnumerator3
   ) {module=ScaleSpriteSize001}
   &ScaleSpriteSize(
    Uniform Curve Scale=0.67
    Uniform Curve Sprite Scale=DataInterface() {propertyValues {Curve {keys [
        {arriveTangent=5.1121, interpMode=RCIM_Cubic, leaveTangent=5.1121, tangentMode=RCTM_User, time=0, value=0}
        {interpMode=RCIM_Cubic, tangentMode=RCTM_User, time=0.4552, value=1.8214}
        {arriveTangent=-0.3706, interpMode=RCIM_Cubic, leaveTangent=-0.3706, tangentMode=RCTM_User, time=0.6326, value=1.836}
        {arriveTangent=1.0993, interpMode=RCIM_Cubic, leaveTangent=1.0993, tangentMode=RCTM_User, time=0.86, value=2.1527}
        <at 1 2.4143>
       ]}}}
   )
   &DynamicMaterialParameters(
    Index 0 Param 1=&FloatFromCurve(FloatCurve=DataInterface() {propertyValues {Curve {keys [
         {arriveTangent=-0.0454, interpMode=RCIM_Cubic, time=0.41, value=0}
         {interpMode=RCIM_Cubic, tangentMode=RCTM_Break, time=0.47, value=0.27}
         {arriveTangent=-0.0928, interpMode=RCIM_Cubic, leaveTangent=-0.0928, time=0.6, value=0.2599}
         <at 0.8 0.2394 RCIM_Cubic>
        ]}}})
    Index 0 Param 2=&FloatFromCurve(FloatCurve=DataInterface() {propertyValues {Curve {keys [
         {arriveTangentWeight=0.05, interpMode=RCIM_Cubic, leaveTangentWeight=0.2833, tangentMode=RCTM_User, time=0, value=0.9}
         {arriveTangentWeight=0.1167, interpMode=RCIM_Cubic, leaveTangentWeight=0.1667, tangentMode=RCTM_Break, time=0.38, value=0.9}
         {arriveTangentWeight=0.1667, interpMode=RCIM_Cubic, tangentMode=RCTM_User, time=0.5, value=-0.9025}
         <at 0.8 -0.1217 RCIM_Cubic>
        ]}}})
    Index 0 Param 3=&FloatFromCurve(FloatCurve=DataInterface() {propertyValues {Curve {keys [
         {arriveTangentWeight=0.1524, interpMode=RCIM_Cubic, leaveTangentWeight=0.3556, tangentMode=RCTM_User, time=0.42, value=0.0974}
         {arriveTangentWeight=0.2333, interpMode=RCIM_Cubic, tangentMode=RCTM_User, time=0.48, value=1}
        ]}}})
    Index 0 Param 4=&FloatFromCurve(FloatCurve=DataInterface() {propertyValues {Curve {keys [<at 0.36 0.14 RCIM_Cubic>,<at 0.45 0>]}}})
    Index 1 Param 1=&FloatFromCurve(FloatCurve=DataInterface() {propertyValues {Curve {keys [<at 0.3719 1 RCIM_Cubic>,<at 0.4809 0>]}}})
    Module.Index0Param0Write=true
    Module.Index0Param1Write=true
    Module.Index0Param2Write=true
    Module.Index0Param3Write=true
    Module.Index1Param0Write=true
    Module.Index1Param1Write=true
    Module.Index1Param2Write=true
    Module.Index1Param3Write=true
    Write Parameter Index 1=true
   )
   &ScaleColor(
    Scale Alpha=&FloatFromCurve(FloatCurve=DataInterface() {propertyValues {Curve {keys [<at 0.73 1 RCIM_Cubic>,<at 1 0>]}}})
    Scale RGB=&VectorFromFloat(Value=&FloatFromCurve(FloatCurve=DataInterface() {propertyValues {Curve {keys [<at 0.2 1 RCIM_Cubic>, {arriveTangent=-0.1429, interpMode=RCIM_Cubic, leaveTangent=-0.1429, time=0.5, value=1.4}, <at 0.9 0.9 RCIM_Cubic>]}}}))
    ScaleA=true
    ScaleRGB=true
   ) {module=ScaleColor001}
   &Drag(Drag=5.5, Use Linear Drag=true, Use Rotational Drag=true)
   &InheritSourceMovement(
    Applied Position Scale=&VectorFromFloat(Value=&MakeCustomFloatFromBool(
      Bool=&SetBoolByFloatComparison(A=&VectorLength(Vector=$Engine.Owner.Velocity), B=600)
      True Float=1
      False Float=0
     ))
    ApplyForce=false
    ApplyPosition=true
    ApplyPositionFalloff=false
   )
   &SolveForcesAndVelocity(Clamp Velocity=false, Limit Acceleration=false)
   /SetVariables(
    Particles.DarkGlow:LinearColor=&MakeCustomLinearColorFromBool(
     Bool=$User.bIsFrontEndPreview
     True Color=&ColorFromCurve(ColorCurve=DataInterface() {propertyValues {
        AlphaCurve {keys [
          <at 0 0 RCIM_Cubic>
          {arriveTangent=0.0218, interpMode=RCIM_Cubic, leaveTangent=0.0218, time=0.0947, value=0.02}
          <at 0.6869 0.015>
          <at 1 0>
         ]}
        BlueCurve {keys [<at 0 0 RCIM_Cubic>]}
        GreenCurve {keys [<at 0 0 RCIM_Cubic>]}
        RedCurve {keys [<at 0 4 RCIM_Cubic>]}
       }})
     False Color=&MultiplyLinearColor(
      Linear Color=&ColorFromCurve(ColorCurve=DataInterface() {propertyValues {
         AlphaCurve {keys [<at 0 0 RCIM_Cubic>,<at 0.0509 0.8 RCIM_Cubic>,<at 0.6396 0.4>,<at 1 0>]}
         BlueCurve {keys [<at 0 0 RCIM_Cubic>]}
         GreenCurve {keys [<at 0 0 RCIM_Cubic>]}
         RedCurve {keys [<at 0 1.3 RCIM_Cubic>]}
        }})
      Scale Factor=&MakeVector4_FromFloat(Float=&FNTimeOfDayScalar(Night Scale=0.6, Morning Scale=1, Evening Scale=0.7))
     )
    )
    Particles.GlowSize=&Multiply_Vector2D(A=$Particles.SpriteSize, B=<1.8 1.8>)
   ) {isSetParameters=true}
   &ScaleColor(
    Scale RGB=&VectorFromFloat(Value=&FNTimeOfDayScalar(Night Scale=0.9, Morning Scale=1, Evening Scale=1))
    ScaleA=true
    ScaleRGB=true
   )
  ]
  rendererClasses [&NiagaraSpriteRendererProperties]
  renderers [
   {rendererClass=&NiagaraSpriteRendererProperties}
   {rendererClass=&NiagaraSpriteRendererProperties, rendererIndex=1}
  ]
 }
 KissSmall {
  emitterUpdateScript [
   &EmitterState(Life Cycle Mode=::NewEnumerator1, Loop Behavior=::NewEnumerator1, Loop Duration=0.3, UseLoopDelay=false)
   &SpawnRate(SpawnRate=25, Use Spawn Probability=false)
  ]
  particleSpawnScript [
   &InitializeParticle(
    Lifetime Max=0.5
    Lifetime Min=0.4
    Lifetime Mode=::NewEnumerator1
    Mass Mode=::NewEnumerator3
    Position Offset=<0 0 5>
    Sprite Rotation Angle=&RandomRangeFloat(Minimum=-15, Maximum=15)
    Sprite Rotation Mode=::NewEnumerator0
    Sprite Size Mode=::NewEnumerator1
    Uniform Sprite Size Max=7
    Uniform Sprite Size Min=4
    UsePositionOffset=true
    Write Material Random=true
   )
   &AddVelocity(Velocity=&RandomRangeVector(Minimum=<-40 30 -40>, Maximum=<40 120 40>), Velocity Speed Scale=1.6)
   /SetVariables(Particles.GlowCameraOffset=-0.1) {isSetParameters=true}
  ]
  particleUpdateScript [
   &ParticleState()
   &DynamicMaterialParameters(
    Index 0 Param 1=&FloatFromCurve(FloatCurve=DataInterface() {propertyValues {Curve {keys [
         <at 0.15 0.2625 RCIM_Cubic>
         {interpMode=RCIM_Cubic, tangentMode=RCTM_Break, time=0.3, value=0.27}
         {arriveTangent=-0.14, interpMode=RCIM_Cubic, leaveTangent=-0.14, time=0.6, value=0.25}
         <at 0.8 0.2 RCIM_Cubic>
        ]}}})
    Index 0 Param 2=&FloatFromCurve(FloatCurve=DataInterface() {propertyValues {Curve {keys [
         {arriveTangentWeight=0.05, interpMode=RCIM_Cubic, leaveTangentWeight=0.2833, tangentMode=RCTM_User, time=0.25, value=0.6758}
         {arriveTangentWeight=0.1167, interpMode=RCIM_Cubic, leaveTangentWeight=0.1667, tangentMode=RCTM_Break, time=0.45, value=0.9}
         {arriveTangentWeight=0.1667, interpMode=RCIM_Cubic, tangentMode=RCTM_User, time=0.8, value=-0.3}
        ]}}})
    Index 0 Param 3=&FloatFromCurve(FloatCurve=DataInterface() {propertyValues {Curve {keys [
         {arriveTangentWeight=0.1524, interpMode=RCIM_Cubic, leaveTangentWeight=0.3556, tangentMode=RCTM_User, time=0.3, value=1}
         {arriveTangentWeight=0.0667, interpMode=RCIM_Cubic, leaveTangentWeight=0.1667, tangentMode=RCTM_Break, time=0.5, value=1}
         {arriveTangentWeight=0.2333, interpMode=RCIM_Cubic, tangentMode=RCTM_User, time=0.8, value=1}
        ]}}})
    Module.Index0Param0Write=true
    Module.Index0Param1Write=true
    Module.Index0Param2Write=true
    Module.Index0Param3Write=true
    Module.Index1Param0Write=true
    Module.Index1Param1Write=true
    Module.Index1Param2Write=true
    Module.Index1Param3Write=true
    Write Parameter Index 1=true
   )
   &ScaleSpriteSize(
    Uniform Curve Scale=0.8
    Uniform Curve Sprite Scale=DataInterface() {propertyValues {Curve {keys [
        {arriveTangent=3.9921, interpMode=RCIM_Cubic, leaveTangent=3.9921, tangentMode=RCTM_User, time=0, value=0}
        {arriveTangent=0.0343, interpMode=RCIM_Cubic, leaveTangent=0.0343, tangentMode=RCTM_User, time=0.6, value=1.875}
        {arriveTangent=-2.6828, interpMode=RCIM_Cubic, leaveTangent=-2.6828, tangentMode=RCTM_User, time=0.85, value=1.5}
        <at 1 0.25>
       ]}}}
   )
   &ScaleSpriteSize(
    Non-Uniform Curve Sprite Scale=DataInterface() {propertyValues {
      XCurve {keys [
        {arriveTangent=-1.9462, interpMode=RCIM_Cubic, leaveTangent=-1.9462, tangentMode=RCTM_User, time=0.1, value=1}
        {arriveTangent=-0.1609, interpMode=RCIM_Cubic, leaveTangent=-0.1609, tangentMode=RCTM_User, time=0.5, value=0.4224}
        {arriveTangent=1.7564, interpMode=RCIM_Cubic, leaveTangent=1.7564, time=0.7552, value=1.15}
        {arriveTangent=1.9401, interpMode=RCIM_Cubic, leaveTangent=1.9401, time=0.9, value=1.125}
        <at 1 1.625 RCIM_Cubic>
       ]}
      YCurve {keys [
        {arriveTangent=-3.0199, interpMode=RCIM_Cubic, leaveTangent=-3.0199, tangentMode=RCTM_User, time=0, value=1.275}
        {arriveTangent=-0.3763, interpMode=RCIM_Cubic, leaveTangent=-0.3763, tangentMode=RCTM_User, time=0.4, value=1}
        <at 0.6 1.25 RCIM_Cubic>
        {arriveTangent=-0.8333, interpMode=RCIM_Cubic, leaveTangent=-0.8333, time=0.7, value=1}
        {arriveTangent=-0.7331, interpMode=RCIM_Cubic, leaveTangent=-0.7331, tangentMode=RCTM_User, time=0.9, value=1}
        <at 1 0 RCIM_Cubic>
       ]}
     }}
    Scale Sprite Size Mode=::NewEnumerator3
   ) {module=ScaleSpriteSize001}
   &ScaleColor(
    Scale Alpha=&FloatFromCurve(FloatCurve=DataInterface() {propertyValues {Curve {keys [<at 0.6 -1 RCIM_Cubic>,<at 1 1>]}}})
    Scale RGB=&VectorFromFloat(Value=&FloatFromCurve(
      FloatCurve=DataInterface() {propertyValues {Curve {keys [<at 0 2.45 RCIM_Cubic>, {arriveTangent=-1.55, interpMode=RCIM_Cubic, leaveTangent=-1.55, time=0.3, value=1.25}, <at 1 0.9 RCIM_Cubic>]}}}
      Scale Curve=2
     ))
    ScaleA=true
    ScaleRGB=true
   ) {module=ScaleColor001}
   &Drag(Drag=2, Use Linear Drag=true, Use Rotational Drag=true)
   &InheritSourceMovement(
    Applied Position Scale=&VectorFromFloat(Value=&MakeCustomFloatFromBool(
      Bool=&SetBoolByFloatComparison(A=&VectorLength(Vector=$Engine.Owner.Velocity), B=600)
      True Float=1
      False Float=0
     ))
    ApplyForce=false
    ApplyPosition=true
    ApplyPositionFalloff=false
   )
   &SolveForcesAndVelocity(Clamp Velocity=false, Limit Acceleration=false)
   /SetVariables(
    Particles.DarkGlow:LinearColor=&MakeCustomLinearColorFromBool(
     Bool=$User.bIsFrontEndPreview
     True Color=&ColorFromCurve(ColorCurve=DataInterface() {propertyValues {
        AlphaCurve {keys [<at 0 0 RCIM_Cubic>, {arriveTangent=0.0412, interpMode=RCIM_Cubic, leaveTangent=0.0412, time=0.1045, value=0.07}, <at 0.7275 0.03>, <at 1 0>]}
        BlueCurve {keys [<at 0 0 RCIM_Cubic>]}
        GreenCurve {keys [<at 0 0 RCIM_Cubic>]}
        RedCurve {keys [<at 0 3 RCIM_Cubic>]}
       }})
     False Color=&ColorFromCurve(ColorCurve=DataInterface() {propertyValues {
        AlphaCurve {keys [<at 0 0 RCIM_Cubic>, {arriveTangent=0.2749, interpMode=RCIM_Cubic, leaveTangent=0.2749, time=0.1045, value=0.3}, <at 0.7275 0.2>, <at 1 0>]}
        BlueCurve {keys [<at 0 0 RCIM_Cubic>]}
        GreenCurve {keys [<at 0 0 RCIM_Cubic>]}
        RedCurve {keys [<at 0 2.4 RCIM_Cubic>]}
       }})
    )
    Particles.GlowSize=&Multiply_Vector2D(A=$Particles.SpriteSize, B=<2.3 2.3>)
   ) {isSetParameters=true}
   &ScaleColor(
    Scale RGB=&VectorFromFloat(Value=&FNTimeOfDayScalar(Night Scale=0.7, Morning Scale=1, Evening Scale=0.8))
    ScaleA=true
    ScaleRGB=true
   )
  ]
  rendererClasses [&NiagaraSpriteRendererProperties]
  renderers [
   {rendererClass=&NiagaraSpriteRendererProperties}
   {rendererClass=&NiagaraSpriteRendererProperties, rendererIndex=1}
  ]
 }
 LightRays {
  emitterUpdateScript [
   &EmitterState(Inactive Response=::NewEnumerator1, Life Cycle Mode=::NewEnumerator1, Loop Behavior=::NewEnumerator1, Loop Duration=1, UseLoopDelay=false)
   &SpawnBurst_Instantaneous(Use Spawn Probability=false, UseLoopCountLimit=false)
  ]
  particleSpawnScript [
   &InitializeParticle(
    Color=<rgba 5 0.2896 0.026 0.85>
    Lifetime=0.65
    Position Offset=<0 -17 0>
    Sprite Rotation Mode=::NewEnumerator2
    Sprite Size=<27 36>
    Sprite Size Mode=::NewEnumerator4
    UsePositionOffset=true
    Write Material Random=true
   )
   &SubUVAnimation(End Frame=3, SubUV Animation Mode:ENiagara_SubUVLookupMode=::NewEnumerator2)
   &AddVelocity(Velocity=<0 1 0>)
  ]
  particleUpdateScript [
   &InheritSourceMovement(
    Applied Position Scale=&VectorFromFloat(Value=&MakeCustomFloatFromBool(
      Bool=&SetBoolByFloatComparison(A=&VectorLength(Vector=$Engine.Owner.Velocity), B=600)
      True Float=1
      False Float=0
     ))
    ApplyForce=false
    ApplyPosition=true
    ApplyPositionFalloff=false
   )
   &SolveForcesAndVelocity(Clamp Velocity=false, Limit Acceleration=false)
   &ParticleState()
   &ScaleSpriteSize(Scale Factor=&MakeVector2D(
     X=&FloatFromCurve(FloatCurve=DataInterface() {propertyValues {Curve {keys [
          {arriveTangent=0.4294, interpMode=RCIM_Cubic, leaveTangent=0.4294, tangentMode=RCTM_User, time=0, value=0.65}
          {interpMode=RCIM_Cubic, tangentMode=RCTM_User, time=1, value=1}
         ]}}})
     Y=&FloatFromCurve(FloatCurve=DataInterface() {propertyValues {Curve {keys [
          {arriveTangent=1.2873, interpMode=RCIM_Cubic, leaveTangent=1.2873, tangentMode=RCTM_User, time=0, value=0}
          {arriveTangent=2.2864, tangentMode=RCTM_User, time=1, value=1.5}
         ]}}})
    ))
   &ScaleColor(
    Scale Alpha=&FloatFromCurve(
     FloatCurve=DataInterface() {propertyValues {Curve {keys [{interpMode=RCIM_Cubic, tangentMode=RCTM_User, time=0, value=1}, {interpMode=RCIM_Cubic, tangentMode=RCTM_User, time=1, value=0}]}}}
     Scale Curve=&FloatFromCurve(FloatCurve=DataInterface() {propertyValues {Curve {keys [
          {interpMode=RCIM_Cubic, tangentMode=RCTM_User, time=0, value=0}
          {interpMode=RCIM_Cubic, tangentMode=RCTM_User, time=0.3, value=1}
          {interpMode=RCIM_Cubic, tangentMode=RCTM_User, time=0.7, value=1}
          {tangentMode=RCTM_User, time=1, value=1}
         ]}}})
    )
    Scale RGB=&MakeVectorFromLinearColorRGB(Color=<rgba 4 4 4 1>)
    ScaleA=true
    ScaleRGB=true
   )
  ]
  rendererClasses [&NiagaraSpriteRendererProperties]
  renderers [{rendererClass=&NiagaraSpriteRendererProperties}]
 }
 Sparkles {
  emitterUpdateScript [
   &EmitterState(Life Cycle Mode=::NewEnumerator1, Loop Behavior=::NewEnumerator1, Loop Duration=0.3, UseLoopDelay=false)
   &SpawnRate(SpawnRate=75, Use Spawn Probability=false)
  ]
  particleSpawnScript [
   /SetVariables(Particles.BiasedRandom=&FloatFromCurve(
     FloatCurve=DataInterface() {propertyValues {Curve {keys [
         <at 0 1 RCIM_Cubic>
         {arriveTangentWeight=0.3486, interpMode=RCIM_Cubic, leaveTangentWeight=0.3486, tangentMode=RCTM_User, time=0.5, value=0}
         <at 1 1 RCIM_Cubic>
        ]}}}
     CurveIndex=&RandomRangeFloat(Minimum=0, Maximum=1)
    )) {isSetParameters=true}
   &InitializeParticle(
    Lifetime Max=0.4
    Lifetime Min=0.3
    Lifetime Mode=::NewEnumerator1
    Sprite Rotation Angle=45
    Sprite Rotation Mode=::NewEnumerator0
    Sprite Size Mode=::NewEnumerator3
    Sprite UV Mode=::NewEnumerator3
    Uniform Sprite Size=&Lerp_Float(A=2, B=4, Alpha=$Particles.BiasedRandom)
    UsePositionOffset=false
    Write Material Random=true
   )
   &ShapeLocation(MaskSpawnGroup=false, Override Seed=false, Sphere Radius=8, Sphere Surface Distribution=1, UseOwnerScale=true)
   &AddVelocity(Velocity=&RandomRangeVector(Minimum=<-40 40 -40>, Maximum=<40 60 40>), Velocity Speed Scale=2.5) {module=AddVelocity001}
  ]
  particleUpdateScript [
   &ParticleState()
   &GravityForce(Gravity=<0 0 -200>)
   &Drag(Use Linear Drag=true, Use Rotational Drag=true)
   &InheritSourceMovement(
    Applied Position Scale=&VectorFromFloat(Value=&MakeCustomFloatFromBool(
      Bool=&SetBoolByFloatComparison(A=&VectorLength(Vector=$Engine.Owner.Velocity), B=600)
      True Float=1
      False Float=0
     ))
    ApplyForce=false
    ApplyPosition=true
    ApplyPositionFalloff=false
   )
   &SolveForcesAndVelocity(Clamp Velocity=false, Limit Acceleration=false)
   &ScaleSpriteSize(Uniform Curve Sprite Scale=DataInterface() {propertyValues {Curve {keys [
        {arriveTangent=12.1401, interpMode=RCIM_Cubic, leaveTangent=12.1401, tangentMode=RCTM_User, time=0, value=0}
        {arriveTangent=1.4286, interpMode=RCIM_Cubic, leaveTangent=1.4286, time=0.1, value=1}
        {arriveTangent=-1.1111, interpMode=RCIM_Cubic, leaveTangent=-1.1111, time=0.7, value=1}
        <at 1 0 RCIM_Cubic>
       ]}}}) {module=ScaleSpriteSize001}
   &ScaleColor(
    Linear Color Curve=DataInterface() {propertyValues {
      AlphaCurve {keys [<at 0.25 1>,<at 1 1>]}
      BlueCurve {keys [<at 0 0.6875>,<at 0.25 0.1023>,<at 1 0>]}
      GreenCurve {keys [<at 0 1.4664>,<at 0.25 0.1819>,<at 1 0.0295>]}
      RedCurve {keys [<at 0 6>,<at 0.25 2>,<at 1 1>]}
     }}
    Scale Mode=::NewEnumerator2
   )
   &ScaleColor(Scale Alpha=&FN_FrontEndFloatModifier(Gameplay=1, Front End=0.4), ScaleA=true, ScaleRGB=true) {module=ScaleColor001}
  ]
  rendererClasses [&NiagaraSpriteRendererProperties]
  renderers [{rendererClass=&NiagaraSpriteRendererProperties}]
 }
}
summary {
 systemName=NS_Emote_Melody_01
 emitters [
  {emitterName=KissBig, rendererClasses [&NiagaraSpriteRendererProperties]}
  {emitterName=KissSmall, rendererClasses [&NiagaraSpriteRendererProperties]}
  {emitterName=Sparkles, rendererClasses [&NiagaraSpriteRendererProperties]}
  {emitterName=LightRays, rendererClasses [&NiagaraSpriteRendererProperties]}
 ]
 userVariables [
  User.bIsFrontEnd {type=NiagaraBool}
  User.bIsFrontEndPreview {type=NiagaraBool}
  User.Color Array {
   type=NiagaraDataInterfaceArrayColor
   description=""
   defaultValue {
    struct=NiagaraExt_VariableValue_DataInterface
    value {
     dataInterface=&self:NiagaraDataInterfaceArrayColor_1
     dataInterfaceClass=&NiagaraDataInterfaceArrayColor
    }
   }
  }
 ]
}
