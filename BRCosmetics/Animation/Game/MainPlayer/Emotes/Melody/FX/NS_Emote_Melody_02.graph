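# Niagara system for the Melody emote that emits a stream of red-glowing kiss sprites with sparkles and a light-ray burst.
# Kisses launch along +Y with curve-animated size and material parameters, while delayed FlareDeath and DecaySparkles emitters spawn at positions sampled from Kisses particles.
# All emitters conditionally inherit owner movement based on a speed threshold.
# Front-end and time-of-day modifiers adjust brightness, glow and flare size.
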NiagaraSystem /BRCosmetics/Animation/Game/MainPlayer/Emotes/Melody/FX/NS_Emote_Melody_02
dependencies {
 usedDataInterfaces [
  &NiagaraDataInterfaceCurve
  &NiagaraDataInterfaceVector2DCurve
  &NiagaraDataInterfaceColorCurve
  &NiagaraDataInterfaceMeshRendererInfo
  &NiagaraDataInterfaceParticleRead
 ]
 usedDynamicInputs [&SelectLinearColorFromArray]
 usedModules [&SystemState]
}
emitters {
 DecaySparkles {
  emitterUpdateScript [
   &EmitterState(Life Cycle Mode=::NewEnumerator1, Loop Behavior=::NewEnumerator1, Loop Delay=0.7, Loop Duration=0.9, UseLoopDelay=true)
   &SpawnRate(SpawnRate=17, Use Spawn Probability=false)
  ]
  particleSpawnScript [
   &FN_SetSpawnGroup(Number of Spawn Groups=1)
   &InitializeParticle(
    Color Maximum=<rgba 1 0.2166 0 1>
    Color Minimum=<rgba 1 0 0.2178 1>
    Color Mode=::NewEnumerator2
    Lifetime=&RandomRangeFloat(Minimum=0.35, Maximum=0.6)
    Sprite Size Mode=::NewEnumerator1
    Uniform Sprite Size Max=40
    Uniform Sprite Size Min=30
    UsePositionOffset=false
    Write Material Random=true
   )
   /SetVariables(Particles.InitialSize:Vector2f=&Vector2DFromFloat(Value=&RandomRangeFloat(Minimum=20, Maximum=50))) {isSetParameters=true}
   /SetVariables(Particles.SpriteRotation2:NiagaraFloat=&RandomRangeFloat(Minimum=0, Maximum=360)) {isSetParameters=true}
   &self:SamplePositionFromOtherParticle(PAR=DataInterface() {propertyValues {EmitterBinding {bindingMode=Other, emitterName=Kisses}}}) {module=SamplePositionFromOtherParticle}
   &ShapeLocation(MaskSpawnGroup=false, Override Seed=false, Sphere Radius=10, UseOwnerScale=true)
   /SetVariables(Particles.LipDeathLoc:NiagaraPosition=$Particles.Position) {isSetParameters=true}
  ]
  particleUpdateScript [
   &ParticleState()
   &InheritSourceMovement(
    Applied Position Scale=&VectorFromFloat(Value=&MakeCustomFloatFromBool(
      Bool=&SetBoolByFloatComparison(A=&VectorLength(Vector=$Engine.Owner.Velocity), B=600)
      True Float=1
      False Float=0
     ))
    ApplyForce=false
    ApplyPosition=true
    ApplyPositionFalloff=false
   )
   /SetVariables(Particles.TimeOfDay:NiagaraFloat=&FNTimeOfDayScalar(Night Scale=0.5, Morning Scale=0.3, Evening Scale=0.3)) {enabled=false, isSetParameters=true}
   /SetVariables(Particles.Color2=&MultiplyLinearColor(
     Linear Color=&ColorFromCurve(ColorCurve=DataInterface() {propertyValues {
        AlphaCurve {keys [<at 0 0>,<at 0.1043 0.2 RCIM_Cubic>,<at 1 0 RCIM_Cubic>]}
        BlueCurve {keys [<at 0.1483 0.0405 RCIM_Cubic>,<at 1 0.0307 RCIM_Cubic>]}
        GreenCurve {keys [<at 0.1483 0.0206 RCIM_Cubic>,<at 1 0.0208 RCIM_Cubic>]}
        RedCurve {keys [<at 0.1483 1 RCIM_Cubic>,<at 1 1 RCIM_Cubic>]}
       }})
     Scale Factor=&MakeFloat4FromVectorAndFloat("Vector (RGB)"=<0.25 0.25 0.25>, "Float (Alpha)"=1)
    )) {isSetParameters=true}
   /SetVariables(Particles.SpriteSize2=&Multiply_Vector2(
     A=$Particles.InitialSize
     B=&Vector2DFromFloat(Value=&FloatFromCurve(
       FloatCurve=DataInterface() {propertyValues {Curve {keys [
           <at 0 0 RCIM_Cubic>
           {arriveTangent=-0.0175, interpMode=RCIM_Cubic, leaveTangent=-0.0175, tangentMode=RCTM_User, time=0.1563, value=0.907}
           {arriveTangent=-0.0042, interpMode=RCIM_Cubic, leaveTangent=-0.0042, tangentMode=RCTM_User, time=0.3, value=0.5}
           {arriveTangent=0.0081, interpMode=RCIM_Cubic, leaveTangent=0.0081, tangentMode=RCTM_User, time=0.4692, value=0.7123}
           {arriveTangent=0.034, interpMode=RCIM_Cubic, leaveTangent=0.034, tangentMode=RCTM_User, time=0.6406, value=0.2751}
           {arriveTangent=0.019, interpMode=RCIM_Cubic, leaveTangent=0.019, tangentMode=RCTM_User, time=0.8, value=0.25}
           <at 1 0 RCIM_Cubic>
          ]}}}
       Scale Curve=0.5
      ))
    )) {isSetParameters=true}
   &ScaleSpriteSize(Scale Factor=&Vector2DFromFloat(Value=&FloatFromCurve(FloatCurve=DataInterface() {propertyValues {Curve {keys [
          <at 0 0 RCIM_Cubic>
          {arriveTangent=-0.0175, interpMode=RCIM_Cubic, leaveTangent=-0.0175, tangentMode=RCTM_User, time=0.1563, value=0.907}
          {arriveTangent=-0.0042, interpMode=RCIM_Cubic, leaveTangent=-0.0042, tangentMode=RCTM_User, time=0.2646, value=0.4751}
          {arriveTangent=0.0081, interpMode=RCIM_Cubic, leaveTangent=0.0081, tangentMode=RCTM_User, time=0.4692, value=0.7123}
          {arriveTangent=0.034, interpMode=RCIM_Cubic, leaveTangent=0.034, tangentMode=RCTM_User, time=0.6406, value=0.4177}
          {arriveTangent=0.019, interpMode=RCIM_Cubic, leaveTangent=0.019, tangentMode=RCTM_User, time=0.7993, value=0.5053}
          <at 1 0 RCIM_Cubic>
         ]}}}))) {module=ScaleSpriteSize001}
   &ScaleColor(
    Scale Alpha=0.6
    Scale RGB=&Multiply_Vector(
     A=<1 1 1>
     B=&VectorFromFloat(Value=&FloatFromCurve(FloatCurve=DataInterface() {propertyValues {Curve {keys [<at 0 0 RCIM_Cubic>,<at 0.24 1 RCIM_Cubic>,<at 1 0>]}}}))
    )
    ScaleA=true
    ScaleRGB=true
   )
   &SolveForcesAndVelocity(Clamp Velocity=false, Limit Acceleration=false)
   &ScaleColor(Scale Alpha=&FN_FrontEndFloatModifier(Gameplay=1, Front End=0.2), ScaleA=true, ScaleRGB=true) {module=ScaleColor001}
  ]
  rendererClasses [&NiagaraSpriteRendererProperties]
  renderers [
   {rendererClass=&NiagaraSpriteRendererProperties}
   {rendererClass=&NiagaraSpriteRendererProperties, rendererIndex=1}
  ]
 }
 FlareDeath {
  emitterUpdateScript [
   &EmitterState(Life Cycle Mode=::NewEnumerator1, Loop Behavior=::NewEnumerator1, Loop Delay=0.7, Loop Duration=0.9, UseLoopDelay=true)
   &SpawnRate(SpawnRate=11, Use Spawn Probability=false)
  ]
  particleSpawnScript [
   &FN_SetSpawnGroup(Number of Spawn Groups=1)
   &InitializeParticle(
    Color=<rgba 2 0 0 0.4>
    Lifetime Max=0.5
    Lifetime Min=0.4
    Lifetime Mode=::NewEnumerator1
    Position Offset Coordinate Space=::Local
    Sprite Rotation Angle=&RandomRangeFloat(Minimum=-10, Maximum=10)
    Sprite Rotation Mode=::NewEnumerator0
    Sprite Size=<60 33>
    Sprite Size Mode=::NewEnumerator4
    UsePositionOffset=true
    Write Material Random=true
   )
   &DynamicMaterialParameters(
    Index 0 Param 1=&RandomRangeFloat(Minimum=0.4, Maximum=0.6)
    Index 0 Param 2=1
    Index 0 Param 3=1
    Index 0 Param 4=1
    Module.Param0WriteEnabled=true
    Module.Param1WriteEnabled=true
    Module.Param2WriteEnabled=true
    Module.Param3WriteEnabled=true
   )
   /SetVariables() {isSetParameters=true}
   &self:SamplePositionFromOtherParticle(PAR=DataInterface() {propertyValues {EmitterBinding {bindingMode=Other, emitterName=Kisses}}}) {module=SamplePositionFromOtherParticle}
  ]
  particleUpdateScript [
   &ParticleState()
   &InheritSourceMovement(
    Applied Position Scale=&VectorFromFloat(Value=&MakeCustomFloatFromBool(
      Bool=&SetBoolByFloatComparison(A=&VectorLength(Vector=$Engine.Owner.Velocity), B=600)
      True Float=1
      False Float=0
     ))
    ApplyForce=false
    ApplyPosition=true
    ApplyPositionFalloff=false
   )
   &SolveForcesAndVelocity(Clamp Velocity=false, Limit Acceleration=false)
   &ScaleColor(
    Scale Alpha=&Multiply_Float(
     A=1
     B=&FloatFromCurve(FloatCurve=DataInterface() {propertyValues {Curve {keys [<at 0 0 RCIM_Cubic>,<at 0.3 1 RCIM_Cubic>,<at 1 0>]}}}, Scale Curve=0.55)
    )
    ScaleA=true
    ScaleRGB=true
   )
   &ScaleSpriteSize(Scale Factor=&Vector2DFromFloat(Value=&FloatFromCurve(
      FloatCurve=DataInterface() {propertyValues {Curve {keys [<at 0 0 RCIM_Cubic>,<at 0.3 1.5 RCIM_Cubic>,<at 1 0>]}}}
      Scale Curve=&FN_FrontEndFloatModifier(Gameplay=1.2, Front End=1)
     )))
   &CameraOffset(Camera Offset Amount=5)
   /SetVariables(Particles.Flash size:Vector2f=&Multiply_Vector2D_ByFloat(
     Float=&Multiply_Float(
      A=40
      B=&FloatFromCurve(
       FloatCurve=DataInterface() {propertyValues {Curve {keys [
           {arriveTangent=4.1785, interpMode=RCIM_Cubic, leaveTangent=4.1785, tangentMode=RCTM_User, time=0, value=0}
           {arriveTangent=0.9995, interpMode=RCIM_Cubic, leaveTangent=0.9995, tangentMode=RCTM_User, time=0.1378, value=0.6825}
           {arriveTangent=-0.0405, interpMode=RCIM_Cubic, leaveTangent=-0.0405, tangentMode=RCTM_User, time=0.62, value=1}
          ]}}}
       Scale Curve=&FN_FrontEndFloatModifier(Gameplay=0.9, Front End=0.7)
      )
     )
     Vector2D=<1.5 1>
    )) {isSetParameters=true}
   /SetVariables(Particles.Flash color:LinearColor=&MakeLinearColorFromVectorAndFloat(
     "Vector (RGB)"=&MakeVectorFromLinearColorRGB(Color=<rgba 1 0.0104 0.0104 0.2>)
     "Float (Alpha)"=&Multiply_Float(
      A=&Multiply_Float(
       A=1
       B=&FloatFromCurve(FloatCurve=DataInterface() {propertyValues {Curve {keys [
            <at 0 0 RCIM_Cubic>
            {interpMode=RCIM_Cubic, tangentMode=RCTM_User, time=0.1, value=1}
            {interpMode=RCIM_Cubic, tangentMode=RCTM_User, time=0.2, value=1}
            <at 0.62 0>
           ]}}})
      )
      B=1
     )
    )) {isSetParameters=true}
   &ScaleColor(Scale Alpha=&FN_FrontEndFloatModifier(Gameplay=1, Front End=0.3), ScaleA=true, ScaleRGB=true) {module=ScaleColor001}
  ]
  rendererClasses [&NiagaraSpriteRendererProperties]
  renderers [{rendererClass=&NiagaraSpriteRendererProperties}]
 }
 Kisses {
  emitterUpdateScript [
   &EmitterState(Life Cycle Mode=::NewEnumerator1, Loop Behavior=::NewEnumerator1, Loop Duration=1, UseLoopDelay=false)
   &SpawnRate(SpawnRate=11, Use Spawn Probability=false)
   &SpawnBurst_Instantaneous(Spawn Group=1, Use Spawn Probability=false, UseLoopCountLimit=false) {enabled=false}
  ]
  particleSpawnScript [
   &InitializeParticle(
    Lifetime Max=0.7
    Lifetime Min=0.7
    Lifetime Mode=::NewEnumerator1
    Mass Mode=::NewEnumerator3
    Position Offset=<0 5 5>
    Sprite Rotation Mode=::NewEnumerator0
    Sprite Size Mode=::NewEnumerator1
    Uniform Sprite Size Max=11
    Uniform Sprite Size Min=9
    UsePositionOffset=true
    Write Material Random=true
   )
   &AddVelocity(Velocity=&RandomRangeVector(Minimum=<-25 150 -25>, Maximum=<25 200 25>))
   /SetVariables(Particles.GlowCameraOffset=-0.1) {isSetParameters=true}
  ]
  particleUpdateScript [
   &ParticleState()
   &SpriteRotationRate(Rotation Rate=&RandomRangeFloat(Minimum=-35, Maximum=35))
   &ScaleSpriteSize(
    Non-Uniform Curve Sprite Scale=DataInterface() {propertyValues {
      XCurve {keys [
        {interpMode=RCIM_Cubic, tangentMode=RCTM_User, time=0.1882, value=0.9232}
        {arriveTangent=0.4905, interpMode=RCIM_Cubic, leaveTangent=0.4905, tangentMode=RCTM_User, time=0.5252, value=0.6426}
        {arriveTangent=1.7384, interpMode=RCIM_Cubic, leaveTangent=1.7384, tangentMode=RCTM_User, time=0.8, value=1.05}
        {interpMode=RCIM_Cubic, leaveTangent=0.1643, time=1, value=0.7319}
       ]}
      YCurve {keys [
        {arriveTangent=-0.6287, interpMode=RCIM_Cubic, leaveTangent=-0.6287, tangentMode=RCTM_User, time=0, value=1.025}
        {arriveTangent=-0.3763, interpMode=RCIM_Cubic, leaveTangent=-0.3763, tangentMode=RCTM_User, time=0.4, value=1}
        {arriveTangent=-0.5, interpMode=RCIM_Cubic, leaveTangent=-0.5, time=0.5769, value=1.1621}
        {arriveTangent=-1.5973, interpMode=RCIM_Cubic, leaveTangent=-1.5973, time=0.8, value=0.8}
        <at 1 0.4863 RCIM_Cubic>
       ]}
     }}
    Scale Sprite Size Mode=::NewEnumerator3
   ) {module=ScaleSpriteSize001}
   &ScaleSpriteSize(
    Uniform Curve Scale=0.77
    Uniform Curve Sprite Scale=DataInterface() {propertyValues {Curve {keys [
        {interpMode=RCIM_Cubic, tangentMode=RCTM_User, time=0, value=0}
        {arriveTangent=1.2783, interpMode=RCIM_Cubic, leaveTangent=1.2783, tangentMode=RCTM_User, time=0.4552, value=1.8214}
        {arriveTangent=-0.3706, interpMode=RCIM_Cubic, leaveTangent=-0.3706, tangentMode=RCTM_User, time=0.5846, value=1.7396}
        {arriveTangent=-2.6828, interpMode=RCIM_Cubic, leaveTangent=-2.6828, tangentMode=RCTM_User, time=0.7207, value=2.0221}
        <at 1 0.25>
       ]}}}
   )
   &DynamicMaterialParameters(
    Index 0 Param 1=&FloatFromCurve(FloatCurve=DataInterface() {propertyValues {Curve {keys [
         {arriveTangent=-0.0454, interpMode=RCIM_Cubic, time=0.41, value=0}
         {interpMode=RCIM_Cubic, tangentMode=RCTM_Break, time=0.5, value=0.27}
         {arriveTangent=-0.1021, interpMode=RCIM_Cubic, leaveTangent=-0.1021, time=0.6, value=0.2599}
         <at 0.8 0.2394 RCIM_Cubic>
        ]}}})
    Index 0 Param 2=&FloatFromCurve(FloatCurve=DataInterface() {propertyValues {Curve {keys [
         {arriveTangentWeight=0.05, interpMode=RCIM_Cubic, leaveTangentWeight=0.2833, tangentMode=RCTM_User, time=0, value=0.9}
         {arriveTangentWeight=0.1167, interpMode=RCIM_Cubic, leaveTangentWeight=0.1667, tangentMode=RCTM_Break, time=0.3297, value=0.9}
         {arriveTangentWeight=0.1667, interpMode=RCIM_Cubic, tangentMode=RCTM_User, time=0.4693, value=-0.9025}
         <at 0.8 -0.1217 RCIM_Cubic>
        ]}}})
    Index 0 Param 3=&FloatFromCurve(FloatCurve=DataInterface() {propertyValues {Curve {keys [
         {arriveTangentWeight=0.1524, interpMode=RCIM_Cubic, leaveTangentWeight=0.3556, tangentMode=RCTM_User, time=0.42, value=0.0974}
         {arriveTangentWeight=0.2333, interpMode=RCIM_Cubic, tangentMode=RCTM_User, time=0.47, value=1}
        ]}}})
    Index 0 Param 4=&FloatFromCurve(FloatCurve=DataInterface() {propertyValues {Curve {keys [<at 0.36 0.14 RCIM_Cubic>,<at 0.4437 0>]}}})
    Index 1 Param 1=&FloatFromCurve(FloatCurve=DataInterface() {propertyValues {Curve {keys [<at 0.3719 1 RCIM_Cubic>,<at 0.4809 0>]}}})
    Module.Index0Param0Write=true
    Module.Index0Param1Write=true
    Module.Index0Param2Write=true
    Module.Index0Param3Write=true
    Module.Index1Param0Write=true
    Module.Index1Param1Write=true
    Module.Index1Param2Write=true
    Module.Index1Param3Write=true
    Write Parameter Index 1=true
   )
   &ScaleColor(
    Scale Alpha=&FloatFromCurve(FloatCurve=DataInterface() {propertyValues {Curve {keys [<at 0.6 1 RCIM_Cubic>,<at 1 1>]}}})
    Scale RGB=&VectorFromFloat(Value=&FloatFromCurve(FloatCurve=DataInterface() {propertyValues {Curve {keys [<at 0 2.45 RCIM_Cubic>, {arriveTangent=-1.55, interpMode=RCIM_Cubic, leaveTangent=-1.55, time=0.3, value=1.25}, <at 1 0.9 RCIM_Cubic>]}}}))
    ScaleA=true
    ScaleRGB=true
   ) {module=ScaleColor001}
   &Drag(Drag=2, Use Linear Drag=true, Use Rotational Drag=true)
   &InheritSourceMovement(
    Applied Position Scale=&VectorFromFloat(Value=&MakeCustomFloatFromBool(
      Bool=&SetBoolByFloatComparison(A=&VectorLength(Vector=$Engine.Owner.Velocity), B=600)
      True Float=1
      False Float=0
     ))
    ApplyForce=false
    ApplyPosition=true
    ApplyPositionFalloff=false
   )
   &SolveForcesAndVelocity(Clamp Velocity=false, Limit Acceleration=false)
   /SetVariables(
    Particles.DarkGlow:LinearColor=&MakeCustomLinearColorFromBool(
     Bool=$User.bIsFrontEndPreview
     True Color=&ColorFromCurve(ColorCurve=DataInterface() {propertyValues {
        AlphaCurve {keys [<at 0 0 RCIM_Cubic>, {arriveTangent=0.0275, interpMode=RCIM_Cubic, leaveTangent=0.0275, time=0.0509, value=0.02}, <at 0.7275 0.02>, <at 1 0>]}
        BlueCurve {keys [<at 0 0 RCIM_Cubic>]}
        GreenCurve {keys [<at 0 0 RCIM_Cubic>]}
        RedCurve {keys [<at 0 2 RCIM_Cubic>]}
       }})
     False Color=&MultiplyLinearColor(
      Linear Color=&ColorFromCurve(ColorCurve=DataInterface() {propertyValues {
         AlphaCurve {keys [<at 0 0 RCIM_Cubic>, {arriveTangent=0.4124, interpMode=RCIM_Cubic, leaveTangent=0.4124, time=0.0509, value=0.3}, <at 0.7275 0.3>, <at 1 0>]}
         BlueCurve {keys [<at 0 0 RCIM_Cubic>]}
         GreenCurve {keys [<at 0 0 RCIM_Cubic>]}
         RedCurve {keys [<at 0 1 RCIM_Cubic>]}
        }})
      Scale Factor=&MakeVector4_FromFloat(Float=&FNTimeOfDayScalar(Night Scale=0.7, Morning Scale=1, Evening Scale=0.8))
     )
    )
    Particles.GlowSize=&Multiply_Vector2D(A=$Particles.SpriteSize, B=<2 2>)
   ) {isSetParameters=true}
   &ScaleColor(
    Scale RGB=&VectorFromFloat(Value=&FNTimeOfDayScalar(Night Scale=0.9, Morning Scale=1, Evening Scale=1))
    ScaleA=true
    ScaleRGB=true
   )
  ]
  rendererClasses [&NiagaraSpriteRendererProperties]
  renderers [
   {rendererClass=&NiagaraSpriteRendererProperties}
   {rendererClass=&NiagaraSpriteRendererProperties, rendererIndex=1}
  ]
 }
 LightRays {
  emitterUpdateScript [
   &EmitterState(Life Cycle Mode=::NewEnumerator1, Loop Behavior=::NewEnumerator1, Loop Duration=1, UseLoopDelay=false)
   &SpawnBurst_Instantaneous(Use Spawn Probability=false, UseLoopCountLimit=false)
  ]
  particleSpawnScript [
   &InitializeParticle(
    Color=<rgba 5 0.2896 0.026 0.85>
    Lifetime=1.1
    Position Offset=<0 -17 0>
    Sprite Rotation Mode=::NewEnumerator2
    Sprite Size=<27 36>
    Sprite Size Mode=::NewEnumerator4
    UsePositionOffset=true
    Write Material Random=true
   )
   &SubUVAnimation(End Frame=3, SubUV Animation Mode:ENiagara_SubUVLookupMode=::NewEnumerator2)
   &AddVelocity(Velocity=<0 1 0>)
  ]
  particleUpdateScript [
   &InheritSourceMovement(
    Applied Position Scale=&VectorFromFloat(Value=&MakeCustomFloatFromBool(
      Bool=&SetBoolByFloatComparison(A=&VectorLength(Vector=$Engine.Owner.Velocity), B=600)
      True Float=1
      False Float=0
     ))
    ApplyForce=false
    ApplyPosition=true
    ApplyPositionFalloff=false
   )
   &SolveForcesAndVelocity(Clamp Velocity=false, Limit Acceleration=false)
   &ParticleState()
   &ScaleSpriteSize(Scale Factor=&MakeVector2D(
     X=&FloatFromCurve(FloatCurve=DataInterface() {propertyValues {Curve {keys [
          {arriveTangent=0.4294, interpMode=RCIM_Cubic, leaveTangent=0.4294, tangentMode=RCTM_User, time=0, value=0.65}
          {interpMode=RCIM_Cubic, tangentMode=RCTM_User, time=1, value=1}
         ]}}})
     Y=&FloatFromCurve(FloatCurve=DataInterface() {propertyValues {Curve {keys [
          {arriveTangent=1.2873, interpMode=RCIM_Cubic, leaveTangent=1.2873, tangentMode=RCTM_User, time=0, value=0}
          {arriveTangent=1.9295, tangentMode=RCTM_User, time=1, value=1.5}
         ]}}})
    ))
   &ScaleColor(
    Scale Alpha=&FloatFromCurve(
     FloatCurve=DataInterface() {propertyValues {Curve {keys [{interpMode=RCIM_Cubic, tangentMode=RCTM_User, time=0, value=1}, {interpMode=RCIM_Cubic, tangentMode=RCTM_User, time=1, value=0}]}}}
     Scale Curve=&FloatFromCurve(FloatCurve=DataInterface() {propertyValues {Curve {keys [
          {interpMode=RCIM_Cubic, tangentMode=RCTM_User, time=0, value=0}
          {interpMode=RCIM_Cubic, tangentMode=RCTM_User, time=0.3, value=1}
          {interpMode=RCIM_Cubic, tangentMode=RCTM_User, time=0.7, value=1}
          {tangentMode=RCTM_User, time=1, value=1}
         ]}}})
    )
    Scale RGB=&MakeVectorFromLinearColorRGB(Color=<rgba 4 4 4 1>)
    ScaleA=true
    ScaleRGB=true
   )
  ]
  rendererClasses [&NiagaraSpriteRendererProperties]
  renderers [{rendererClass=&NiagaraSpriteRendererProperties}]
 }
 Sparkles {
  emitterUpdateScript [
   &EmitterState(Life Cycle Mode=::NewEnumerator1, Loop Behavior=::NewEnumerator1, Loop Duration=1, UseLoopDelay=false)
   &SpawnRate(SpawnRate=20, Use Spawn Probability=false)
  ]
  particleSpawnScript [
   /SetVariables(Particles.BiasedRandom=&FloatFromCurve(
     FloatCurve=DataInterface() {propertyValues {Curve {keys [
         <at 0 1 RCIM_Cubic>
         {arriveTangentWeight=0.3486, interpMode=RCIM_Cubic, leaveTangentWeight=0.3486, tangentMode=RCTM_User, time=0.5, value=0}
         <at 1 1 RCIM_Cubic>
        ]}}}
     CurveIndex=&RandomRangeFloat(Minimum=0, Maximum=1)
    )) {isSetParameters=true}
   &InitializeParticle(
    Lifetime Max=0.8
    Lifetime Min=0.5
    Lifetime Mode=::NewEnumerator1
    Sprite Rotation Angle=45
    Sprite Rotation Mode=::NewEnumerator0
    Sprite Size Mode=::NewEnumerator3
    Sprite UV Mode=::NewEnumerator3
    Uniform Sprite Size=&Lerp_Float(A=2, B=4, Alpha=$Particles.BiasedRandom)
    UsePositionOffset=false
    Write Material Random=true
   )
   &ShapeLocation(MaskSpawnGroup=false, Override Seed=false, Sphere Radius=6, Sphere Surface Distribution=1, UseOwnerScale=true)
   &AddVelocity(Velocity=&RandomRangeVector(Minimum=<-30 30 -30>, Maximum=<30 60 30>), Velocity Speed Scale=2.5) {module=AddVelocity001}
  ]
  particleUpdateScript [
   &ParticleState()
   &Drag(Use Linear Drag=true, Use Rotational Drag=true)
   &InheritSourceMovement(
    Applied Position Scale=&VectorFromFloat(Value=&MakeCustomFloatFromBool(
      Bool=&SetBoolByFloatComparison(A=&VectorLength(Vector=$Engine.Owner.Velocity), B=600)
      True Float=1
      False Float=0
     ))
    ApplyForce=false
    ApplyPosition=true
    ApplyPositionFalloff=false
   )
   &SolveForcesAndVelocity(Clamp Velocity=false, Limit Acceleration=false)
   &ScaleSpriteSize(Uniform Curve Sprite Scale=DataInterface() {propertyValues {Curve {keys [
        {arriveTangent=12.1401, interpMode=RCIM_Cubic, leaveTangent=12.1401, tangentMode=RCTM_User, time=0, value=0}
        {arriveTangent=1.4286, interpMode=RCIM_Cubic, leaveTangent=1.4286, time=0.1, value=1}
        {arriveTangent=-1.1111, interpMode=RCIM_Cubic, leaveTangent=-1.1111, time=0.7, value=1}
        <at 1 0 RCIM_Cubic>
       ]}}}) {module=ScaleSpriteSize001}
   &ScaleColor(
    Linear Color Curve=DataInterface() {propertyValues {
      AlphaCurve {keys [<at 0.25 1>,<at 1 1>]}
      BlueCurve {keys [<at 0 0.6875>,<at 0.25 0.1023>,<at 1 0>]}
      GreenCurve {keys [<at 0 1.4664>,<at 0.25 0.1819>,<at 1 0.0295>]}
      RedCurve {keys [<at 0 6>,<at 0.25 2>,<at 1 1>]}
     }}
    Scale Mode=::NewEnumerator2
   )
   &ScaleColor(Scale Alpha=&FN_FrontEndFloatModifier(Gameplay=1, Front End=0.2), ScaleA=true, ScaleRGB=true) {module=ScaleColor001}
  ]
  rendererClasses [&NiagaraSpriteRendererProperties]
  renderers [{rendererClass=&NiagaraSpriteRendererProperties}]
 }
}
summary {
 systemName=NS_Emote_Melody_02
 emitters [
  {emitterName=Kisses, rendererClasses [&NiagaraSpriteRendererProperties]}
  {emitterName=Sparkles, rendererClasses [&NiagaraSpriteRendererProperties]}
  {emitterName=LightRays, rendererClasses [&NiagaraSpriteRendererProperties]}
  {emitterName=FlareDeath, rendererClasses [&NiagaraSpriteRendererProperties]}
  {emitterName=DecaySparkles, rendererClasses [&NiagaraSpriteRendererProperties]}
 ]
 userVariables [
  User.bIsFrontEnd {type=NiagaraBool}
  User.bIsFrontEndPreview {type=NiagaraBool}
  User.Color Array {
   type=NiagaraDataInterfaceArrayColor
   description=""
   defaultValue {
    struct=NiagaraExt_VariableValue_DataInterface
    value {
     dataInterface=&self:NiagaraDataInterfaceArrayColor_0
     dataInterfaceClass=&NiagaraDataInterfaceArrayColor
    }
   }
  }
 ]
}
