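# Niagara system for the Melody emote that puffs dust sprites from the character's feet.
# One emitter reuses SeleneCobra's SpawnFootDust module, which samples the skeletal mesh against a user height threshold and spawns rate-limited bursts at either foot.
# Particles start on a ring near owner height, move along its bitangent, then drag, rise and spin relative to camera.
# Curves animate size, material parameters and time-of-day-scaled color, while front-end modifiers add local offset and acceleration.
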NiagaraSystem /BRCosmetics/Animation/Game/MainPlayer/Emotes/Melody/FX/NS_Emote_Melody_DustFeet
dependencies {
 usedDataInterfaces [
  &NiagaraDataInterfaceCurve
  &NiagaraDataInterfaceCamera
  &NiagaraDataInterfaceColorCurve
  &NiagaraDataInterfaceSkeletalMesh
 ]
 usedDynamicInputs [
  &Vector2DFromFloat
  &FNTimeOfDayScalar
  /Game/Effects/Niagara/DynamicInputs/State/FN_IsExecutionStateActive
 ]
 usedModules [
  &self:SystemSpawnScript.SystemScriptSource.SystemScriptGraph.NiagaraNodeAssignment_2.SetVariables_#1
  &SystemState
  &self:SystemSpawnScript.SystemScriptSource.SystemScriptGraph.NiagaraNodeAssignment_1.SetVariables_#2
  &Timeline
 ]
}
emitters {DustPuffFeet {
  emitterUpdateScript [
   &EmitterState(Life Cycle Mode=::NewEnumerator1, Loop Duration=1, UseLoopDelay=false)
   /Game/Animation/Game/MainPlayer/Emotes/SeleneCobra/FX/SpawnFootDust(
    Height Threshold For Spawning:NiagaraFloat=$User.HeightThreshold
    Minimum Interval Between Spawn Events=0.2
    Right Foot Bone Index=1
    Skeletal Mesh:NiagaraDataInterfaceSkeletalMesh=$System.Skeletal Mesh
    Spawn Count=6
   )
  ]
  particleSpawnScript [
   &InitializeParticle(
    Color=<rgba 0.1 0.0928 0.084 0.2>
    Lifetime Max=1
    Lifetime Min=0.6
    Lifetime Mode=::NewEnumerator1
    Position=&MaskVectorBySpawnGroup(
     Masked Vector=Linked() {linkedVariable {name=Emitter.LeftFootSpawnPositionWorldSpace, type=NiagaraPosition}}
     Passthrough Vector=Linked() {linkedVariable {name=Emitter.RightFootSpawnPositionWorldSpace, type=NiagaraPosition}}
    )
    Position Mode=::NewEnumerator1
    Position Offset=&FN_FrontEndVector3Modifier(Gameplay=<0 0 0>, Front End=<0 -16 0>)
    Position Offset Coordinate Space=::Local
    Sprite Rotation Mode=::NewEnumerator2
    Sprite Size Mode=::NewEnumerator1
    Uniform Sprite Size Max=44
    Uniform Sprite Size Min=33
    UsePositionOffset=true
    Write Material Random=true
   )
   &ShapeLocation(
    MaskSpawnGroup=false
    Override Seed=false
    Ring / Disc Distribution Mode:ENiagaraSphereDistributionMode=::NewEnumerator1
    Ring Radius=11
    Shape Origin=&MakeVector(
     X=&MakeFloatFromVector(Vector=Linked() {linkedVariable {name=StackContext.Position, type=NiagaraPosition}}, Channel=::NewEnumerator0)
     Y=&MakeFloatFromVector(Vector=Linked() {linkedVariable {name=StackContext.Position, type=NiagaraPosition}}, Channel=::NewEnumerator1)
     Z=&Add_Float(
      A=5
      B=&MakeFloatFromVector(Vector=Linked() {linkedVariable {name=Engine.Owner.Position, type=NiagaraPosition}}, Channel=::NewEnumerator2)
     )
    )
    Shape Primitive=::NewEnumerator4
    U Position=&ReturnNormalizedExecIndex()
    UseOwnerScale=false
   )
   &AddVelocity(
    Rotation Coordinate Space=::World
    Velocity=$Particles.ShapeLocation.ShapeBitangent
    Velocity Speed Scale=&RandomRangeFloat(Minimum=111, Maximum=222)
   )
  ]
  particleUpdateScript [
   &CameraOffset(Camera Offset Amount=&FloatFromCurve(
     FloatCurve=DataInterface() {propertyValues {Curve {keys [{tangentMode=RCTM_User, time=0, value=0}, {tangentMode=RCTM_User, time=1, value=1}]}}}
     Scale Curve=44
    ))
   &Drag(Drag=6, Rotational Drag=0, Use Linear Drag=true, Use Rotational Drag=false)
   &GravityForce(Gravity=<0 0 33>)
   &ParticleState()
   &AccelerationForce(
    Acceleration=&FN_FrontEndVector3Modifier(Gameplay=<0 0 0>, Front End=<0 -333 0>)
    Coordinate Space=::Local
   )
   &SolveForcesAndVelocity(Clamp Velocity=false, Limit Acceleration=false)
   &DynamicMaterialParameters(
    Index 0 Param 1=&Add_Float(
     A=&RandomRangeFloat(Minimum=0, Maximum=1)
     B=&FloatFromCurve(
      FloatCurve=DataInterface() {propertyValues {Curve {keys [
          {arriveTangent=0.9088, interpMode=RCIM_Cubic, leaveTangent=0.9088, tangentMode=RCTM_User, time=0, value=1}
          {arriveTangent=0.5473, interpMode=RCIM_Cubic, leaveTangent=0.5473, tangentMode=RCTM_User, time=1, value=2}
         ]}}}
      CurveIndex=$Particles.Age
      Scale Curve=0.222
     )
    )
    Index 0 Param 2=&FloatFromCurve(FloatCurve=DataInterface() {propertyValues {Curve {keys [
         {arriveTangent=0.0453, interpMode=RCIM_Cubic, tangentMode=RCTM_Break, time=0, value=-1}
         {arriveTangent=0.0604, interpMode=RCIM_Cubic, leaveTangent=0.0604, tangentMode=RCTM_User, time=1, value=-1}
        ]}}})
    Index 0 Param 3=&FloatFromCurve(
     FloatCurve=DataInterface() {propertyValues {Curve {keys [
         {arriveTangent=-9.0447, interpMode=RCIM_Cubic, leaveTangent=-2.0044, tangentMode=RCTM_User, time=0, value=0.5}
         {interpMode=RCIM_Cubic, leaveTangent=4.6779, time=0.2851, value=0}
        ]}}}
     Scale Curve=0
    )
    Index 1 Param 1=&FloatFromCurve(
     FloatCurve=DataInterface() {propertyValues {Curve {keys [<at 0 0>,<at 0.25 1>]}}}
     Scale Curve=&RandomRangeFloat(Minimum=2, Maximum=1)
    )
    Index 1 Param 2=&FloatFromCurve(
     FloatCurve=DataInterface() {propertyValues {Curve {keys [
         {interpMode=RCIM_Cubic, leaveTangent=3.7422, tangentMode=RCTM_Break, time=0, value=0.2}
         {interpMode=RCIM_Cubic, leaveTangent=0.9066, time=1, value=2}
        ]}}}
     Scale Curve=&RandomRangeFloat(Minimum=0.5, Maximum=1)
    )
    Index 1 Param 3=&FloatFromCurve(FloatCurve=DataInterface() {propertyValues {Curve {keys [
         {interpMode=RCIM_Cubic, leaveTangent=2.8445, tangentMode=RCTM_Break, time=0, value=0.2}
         {interpMode=RCIM_Cubic, leaveTangent=63.2978, time=0.25, value=1}
        ]}}})
    Index 1 Param 4=&FloatFromCurve(
     FloatCurve=DataInterface() {propertyValues {Curve {keys [<at 0 1>,<at 1 0.8>]}}}
     Scale Curve=&RandomRangeFloat(Minimum=0.5, Maximum=1)
    )
    Index 2 Param 1=&FloatFromCurve(
     FloatCurve=DataInterface() {propertyValues {Curve {keys [{interpMode=RCIM_Cubic, leaveTangent=3.2309, tangentMode=RCTM_Break, time=0.1, value=0}, <at 1 1>]}}}
     Scale Curve=0
    )
    Index 2 Param 2=&FloatFromCurve(
     FloatCurve=DataInterface() {propertyValues {Curve {keys [<at 0 0.2 RCIM_Cubic>,<at 1 0.8 RCIM_Cubic>]}}}
     Scale Curve=&RandomRangeFloat(Minimum=0.2, Maximum=1.2)
    )
    Index 2 Param 3=&FloatFromCurve(FloatCurve=DataInterface() {propertyValues {Curve {keys [
         {interpMode=RCIM_Cubic, tangentMode=RCTM_Break, time=0, value=1}
         {interpMode=RCIM_Cubic, leaveTangent=11.1233, tangentMode=RCTM_Break, time=1, value=2}
        ]}}})
    Index 2 Param 4=&FloatFromCurve(
     FloatCurve=DataInterface() {propertyValues {Curve {keys [<at 0 1 RCIM_Cubic>, {interpMode=RCIM_Cubic, leaveTangent=-1.2956, tangentMode=RCTM_Break, time=1, value=0.75}]}}}
     Scale Curve=0.8
    )
    Module.Index0Param0Write=true
    Module.Index0Param1Write=true
    Module.Index0Param2Write=true
    Module.Index0Param3Write=true
    Module.Index1Param0Write=true
    Module.Index1Param1Write=true
    Module.Index1Param2Write=true
    Module.Index1Param3Write=true
    Module.Index2Param0Write=true
    Module.Index2Param1Write=true
    Module.Index2Param2Write=true
    Module.Index2Param3Write=true
    Write Parameter Index 1=true
    Write Parameter Index 2=true
   )
   &ScaleSpriteSize(Uniform Curve Sprite Scale=DataInterface() {propertyValues {Curve {keys [
        {arriveTangent=4.1785, interpMode=RCIM_Cubic, leaveTangent=4.1785, tangentMode=RCTM_User, time=0, value=0}
        {arriveTangent=1.1179, interpMode=RCIM_Cubic, leaveTangent=1.1179, tangentMode=RCTM_User, time=0.2042, value=0.6228}
        {arriveTangent=0.0224, interpMode=RCIM_Cubic, leaveTangent=0.0224, tangentMode=RCTM_User, time=1, value=1}
       ]}}})
   &SpriteRotationRate(Rotation Rate=&Multiply_Float(
     A=&Lerp_Float(A=-555, B=-55, Alpha=$Particles.NormalizedAge)
     B=&DotProduct(
      Vector A=&ReturnCameraVector(Camera Vectors=::NewEnumerator3, Coordinate Space=::World)
      Vector B=$Particles.Initial.ShapeLocation.ShapeBitangent
     )
    ))
   &ScaleColor(
    Scale Alpha=&FloatFromCurve(FloatCurve=DataInterface() {propertyValues {Curve {keys [
         {interpMode=RCIM_Cubic, leaveTangent=13.0812, tangentMode=RCTM_Break, time=0, value=0}
         <at 0.05 1 RCIM_Cubic>
         {arriveTangent=-0.3298, interpMode=RCIM_Cubic, leaveTangent=-0.3298, tangentMode=RCTM_User, time=1, value=0}
        ]}}})
    Scale RGB=&VectorFromFloat(Value=&FloatFromCurve(FloatCurve=DataInterface() {propertyValues {Curve {keys [<at 0 1>,<at 0.9983 0.5>]}}}, Scale Curve=$System.TimeOfDayBrightnessMultiplier))
    ScaleA=true
    ScaleRGB=true
   )
   /SetVariables(Particles.ColorUnlit:LinearColor=&MultiplyLinearColor(
     Linear Color=&ColorFromCurve(ColorCurve=DataInterface() {propertyValues {
        AlphaCurve {keys [<at 0 0 RCIM_Cubic>,<at 0.1946 1>,<at 1 0 RCIM_Cubic>]}
        BlueCurve {keys [<at 0.1113 0.084>,<at 1 0.0504 RCIM_Cubic>]}
        GreenCurve {keys [<at 0.1113 0.0928>,<at 1 0.0557 RCIM_Cubic>]}
        RedCurve {keys [<at 0.1113 0.1>,<at 1 0.06 RCIM_Cubic>]}
       }})
     Scale Factor=&MakeVector4_FromFloat(Float=$System.TimeOfDayBrightnessMultiplier)
    )) {isSetParameters=true}
  ]
  rendererClasses [&NiagaraSpriteRendererProperties]
  renderers [
   {rendererClass=&NiagaraSpriteRendererProperties}
   {rendererClass=&NiagaraSpriteRendererProperties, rendererIndex=1}
  ]
 }}
summary {
 systemName=NS_Emote_Melody_DustFeet
 emitters [{emitterName=DustPuffFeet, rendererClasses [&NiagaraSpriteRendererProperties]}]
 userVariables [
  User.bIsFrontEnd {type=NiagaraBool}
  User.bIsFrontEndPreview {type=NiagaraBool}
  User.HeightThreshold {type=NiagaraFloat, defaultValue=0, description=""}
 ]
}
